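# Material function for the Rafael vehicle body's Livery2 skin, which paints a red, view-dependent body color over the shared MF_Vehicle_Body_Base_v2 attributes.
# Nested lerps blend dark-to-bright reds through biased, scaled dot-product rings, apparently of camera vector and vertex normal, as a fake Fresnel falloff.
# A tiled sparkle-flake noise is declared, likely breaking up some ring masks.
# A final bias-scale boosts contrast before the saturated result replaces BaseColor.
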MaterialFunction /VehicleCosmetics/Bodies/Rafael/Skins/Livery2/MF_Skin_Rafael_Livery2
graph {nodes [
  n1 = Constant3Vector() {constant=<rgba 0.1476 0.0018 0.0042 1>}
  n2 = Constant3Vector() {desc=Primary body color, constant=<rgba 0.658 0.0248 0.0248 0>}
  n3 = Constant3Vector() {desc=Shine Highlight, constant=<rgba 0.717 0.1475 0.0802 1>}
  n4 = Multiply(B=NamedRerouteUsage()) {constA=0.87}
  n5 = NamedRerouteUsage()
  n6 = Saturate(None=ConstantBiasScale(None=DotProduct(A=@n4, B=@n5)) {bias=-0.859, desc="Bias == Size of Ring; Scale == Strength of contrast", scale=2})
  n7 = Multiply(B=NamedRerouteUsage()) {constA=0.86}
  n8 = NamedRerouteUsage()
  n9 = Add(
   A=ConstantBiasScale(None=DotProduct(A=@n7, B=@n8)) {bias=-0.725, desc="Bias == Size of Ring; Scale == Strength of contrast", scale=5}
   B=ConstantBiasScale(None=NamedRerouteUsage()) {bias=-0.1, scale=2}
  )
  n10 = DotProduct(A=Multiply(B=NamedRerouteUsage()) {constA=0.97}, B=NamedRerouteUsage())
  n11 = LinearInterpolate(
   A=Constant3Vector() {constant=<rgba 0.2795 0.0067 0.0041 1>}
   B=LinearInterpolate(A=Constant3Vector() {constant=<rgba 0.4497 0.0153 0.0065 1>}, B=LinearInterpolate(A=@n2, B=@n3, Alpha=@n6), Alpha=Saturate(None=@n9))
   Alpha=Saturate(None=ConstantBiasScale(None=@n10) {bias=-0.68, desc="Bias == Size of Ring; Scale == Strength of contrast", scale=7})
  )
  n12 = DotProduct(A=Multiply(B=NamedRerouteUsage()) {constA=0.62}, B=NamedRerouteUsage())
  n13 = NamedRerouteUsage()
  n14 = Saturate(None=Add(
    A=ConstantBiasScale(None=@n12) {bias=-0.2, desc="Bias == Size of Ring; Scale == Strength of contrast", scale=4}
    B=ConstantBiasScale(None=@n13) {bias=-0.01, scale=1}
   ))
  n15 = NamedRerouteUsage()
  n16 = ConstantBiasScale(None=DotProduct(A=Multiply(B=@n15) {constA=0.68}, B=NamedRerouteUsage())) {bias=-0.001, desc="Bias == Size of Ring; Scale == Strength of contrast", scale=6}
  n17 = ConstantBiasScale(None=LinearInterpolate(A=Constant3Vector() {constant=<rgba 0.14 0.0141 0.0037 0>}, B=LinearInterpolate(A=@n1, B=@n11, Alpha=@n14), Alpha=Saturate(None=@n16))) {bias=-0.01, desc="Lower Bias == Higher Contrast when \r\npaired with brighter Scale", scale=1.8}
  FunctionOutput(None=SetMaterialAttributes!(
    MaterialAttributes=MaterialFunctionCall() {materialFunction=/VehicleCosmetics/Materials/VehicleParents_Version2/Material_Functions/MF_Vehicle_Body_Base_v2}.Result
    BaseColor=Saturate(None=@n17)
   ) {attributeSetTypes [#1]})
  NamedRerouteDeclaration(None=TextureSample(UVs=TextureCoordinate() {uTiling=50, vTiling=50}) {texture=/VehicleCosmetics/SharedTextures/T_Noise_Sparkle_Flake}.RGB) {name=SparkleMask_Out, nodeColor=<rgba 0 1 0.2031 1>}
  NamedRerouteDeclaration(None=CameraVectorWS()) {name=CameraVector_Out, nodeColor=<rgba 1 0 0.1172 1>}
  NamedRerouteDeclaration(None=VertexNormalWS()) {name=VNormal_Out, nodeColor=<rgba 0 1 0.4141 1>}
 ]}
